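; Apollo_IceCube_Hotel is an ice cube prop Blueprint for the Ice Hotel environment, with a point light and an optional cosmetic mesh.
; The construction script adds a collision-free static mesh at a configurable relative transform when CosmeticMesh is valid, and sets light visibility from forceLightsOff.
; On BeginPlay, after the parent call, forceLightsOff also hides the light, disables its shadow casting and zeroes its intensity.
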(event EventBeginPlay
  (bind _pointlight (Variables|Default|GetPointLight))
  (|CallParentFunction)
  (if (Variables|Default|GetforceLightsOff)
    (Rendering|SetVisibility _pointlight false false)
    (Rendering|Components|Light|SetCastShadows _pointlight false)
    (Rendering|Components|Light|SetIntensity _pointlight 0.0)))

(fn ConstructionScript
  (Utilities|FlowControl|Sequence
    (:then_0
      (Utilities|IsValid (Variables|Default|GetCosmeticMesh)
        (:Is%20Valid
          (bind _returnvalue (AddComponent|NODE_AddStaticMeshComponent-0 :RelativeTransform (Variables|Default|GetCosmeticMeshRelativeTransform)))
          (Components|StaticMesh|SetStaticMesh _returnvalue (Variables|Default|GetCosmeticMesh))
          (Collision|SetCollisionEnabled _returnvalue "NoCollision"))))
    (:then_1
      (Rendering|SetVisibility (Variables|Default|GetPointLight) (not (Variables|Default|GetforceLightsOff)) false))))
