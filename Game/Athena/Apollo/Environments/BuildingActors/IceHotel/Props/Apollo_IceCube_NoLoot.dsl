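; A destructible ice cube building prop for the Ice Hotel environment, named as the no-loot variant, that reacts visually to damage.
; On damage it spawns frozen-loot impact particles at the hit point and actor origin.
; It also feeds lost health fraction, game time and hit location into the static mesh materials for damage and wobble effects.
; The construction script adds an optional cosmetic mesh at a configurable relative transform with collision disabled.
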(event EventOnDamagePlayEffects
  (params (Damage :type Float%20%28single-precision%29) (DamageTags :type Gameplay%20Tag%20Container%20Structure%20%28by%20ref%29) (Momentum :type Vector :default "0, 0, 0") (HitInfo :type Hit%20Result%20Structure%20%28by%20ref%29) (InstigatedBy :type Fort%20Pawn%20Object%20Reference) (DamageCauser :type Actor%20Object%20Reference) (EffectContext :type Gameplay%20Effect%20Context%20Handle%20Structure))
  (bind (_bblockinghit _binitialoverlap _time _distance _location _impactpoint _normal _impactnormal) (Collision|BreakHitResult HitInfo))
  (bind _staticmeshcomponent (Variables|Debug|GetStaticMeshComponent))
  ;[1 1 1] Impact Effects
  (Effects|SpawnEmitteratLocation "/Game/Athena/Items/Gameplay/FrozenLoot/P_FrozenLoot_Impact.P_FrozenLoot_Impact" _location (Math|Conversions|RotationFromXVector _impactnormal))
  (Effects|SpawnEmitteratLocation "/Game/Athena/Items/Gameplay/FrozenLoot/P_FrozenLoot_ImpactExtra.P_FrozenLoot_ImpactExtra" (Transformation|GetActorLocation))
  ;[1 1 1] Update material damage level
  (Rendering|Material|SetScalarParameterValueonMaterials _staticmeshcomponent "BP_DamageLevel" (- 1.0 (Building|GetBuildingHealthPercentage self)))
  (Rendering|Material|SetScalarParameterValueonMaterials _staticmeshcomponent "BP_WobbleTime" (Utilities|Time|GetGameTimeinSeconds))
  (Rendering|Material|SetVectorParameterValueonMaterials _staticmeshcomponent "HitLocation" _location))

(fn ConstructionScript
  (Utilities|IsValid (Variables|Default|GetCosmeticMesh)
    (:Is%20Valid
      (bind _returnvalue (AddComponent|NODE_AddStaticMeshComponent-0 :RelativeTransform (Variables|Default|GetCosmeticMeshRelativeTransform)))
      (Components|StaticMesh|SetStaticMesh _returnvalue (Variables|Default|GetCosmeticMesh))
      (Collision|SetCollisionEnabled _returnvalue "NoCollision"))))
