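; Building actor prop blueprint for the Shiver Inn in the Apollo IceHotel environment set, with support for an optional decorative mesh.
; When CosmeticMesh is valid, the construction script adds a static mesh component at CosmeticMeshRelativeTransform and assigns that mesh to it.
; Collision is disabled on the added component, so the extra mesh is purely visual.
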(fn ConstructionScript
  (Utilities|IsValid (Variables|Default|GetCosmeticMesh)
    (:Is%20Valid
      (bind _returnvalue (AddComponent|NODE_AddStaticMeshComponent-0 :RelativeTransform (Variables|Default|GetCosmeticMeshRelativeTransform)))
      (Components|StaticMesh|SetStaticMesh _returnvalue (Variables|Default|GetCosmeticMesh))
      (Collision|SetCollisionEnabled _returnvalue "NoCollision"))))
